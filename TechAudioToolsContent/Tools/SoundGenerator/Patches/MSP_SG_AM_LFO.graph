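# MetaSound patch for the Sound Generator tool that applies LFO amplitude modulation to a float amplitude signal.
# The Frequency input is range-mapped, topping out at 40, to drive an LFO swinging between 0 and -1.
# That LFO is scaled by Depth and AmpSignal and added to AmpSignal, so modulation only attenuates.
# A second range map flips the LFO positive for the separate LFO.Output.
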MetaSoundPatch /TechAudioToolsContent/Tools/SoundGenerator/Patches/MSP_SG_AM_LFO
inputs [
 AmpSignal {type=Float}
 Depth {type=Float}
 Frequency {type=Float}
]
outputs [
 LFO.Output {type=Float}
 Out {type=Float}
]
pages [Default {
  input_defaults {AmpSignal=0, Depth=0, Frequency=0}
  nodes [
   n1 = UE.LFO:Audio(Frequency=MapRange.MapRange:Float(In=^Frequency) {also {defaults {Out Range B=40}}}.Out Value) {also {defaults {Max Value=-1, Min Value=0}}}
   Add_0 = UE.Add:Float(
    PrimaryOperand=^AmpSignal
    AdditionalOperands=UE.Multiply:Float(PrimaryOperand=^AmpSignal, AdditionalOperands=UE.Multiply:Float(AdditionalOperands=^Depth, PrimaryOperand=@n1.Out).Out).Out
   )
   MapRangeMapRange_1 = MapRange.MapRange:Float(In=@n1.Out) {also {defaults {In Range B=-1}}}
  ]
  links [{dst="@Out", src=Add_0.Out}, {dst="@LFO.Output", src=MapRangeMapRange_1.Out Value}]
 }]
