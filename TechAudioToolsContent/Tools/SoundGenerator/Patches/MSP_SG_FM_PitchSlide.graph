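# MetaSound patch for the SoundGenerator tool that produces a triggered, audio-rate pitch slide offset, named for FM use.
# PitchTarget maps to a range between negative and positive BaseFrequency, which is also output as Range.
# OnPlay fires a zero-attack AD envelope decaying over Duration, inverted so Out Value slides from zero to that range.
# PitchCurve shapes the decay through a remap and fourth-power response.
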MetaSoundPatch /TechAudioToolsContent/Tools/SoundGenerator/Patches/MSP_SG_FM_PitchSlide
inputs [
 BaseFrequency {type=Float}
 Duration {type=Time}
 OnPlay {type=Trigger}
 PitchCurve {type=Float}
 PitchTarget {type=Float}
]
outputs [
 Out Value {type=Audio}
 Range {type=Float}
]
pages [Default {
  input_defaults {BaseFrequency=0, Duration=0, OnPlay=null, PitchCurve=0, PitchTarget=0}
  nodes [
   ADEnvelopeADEnvelope_0 = AD Envelope.AD Envelope:Audio(
    Decay Time=^Duration
    Trigger=^OnPlay
    Decay Curve=UE.Power:Float(PrimaryOperand=MapRange.MapRange:Float(In=^PitchCurve) {also {defaults {Out Range A=0.5, Out Range B=1.5}}}.Out Value) {also {defaults {AdditionalOperands=4}}}.Out
   ) {also {defaults {Attack Time=0}}}
   MapRangeMapRange_2 = MapRange.MapRange:Float(
    Out Range B=^BaseFrequency
    In=^PitchTarget
    Out Range A=UE.Multiply:Float(PrimaryOperand=^BaseFrequency) {also {defaults {AdditionalOperands=-1}}}.Out
   )
   MapRangeMapRange_0 = MapRange.MapRange:Audio(In=@ADEnvelopeADEnvelope_0.Out Envelope, Out Range B=@MapRangeMapRange_2.Out Value) {also {defaults {In Range A=1, In Range B=0, Out Range A=0}}}
  ]
  links [{dst="@Out Value", src=MapRangeMapRange_0.Out Value}, {dst="@Range", src=MapRangeMapRange_2.Out Value}]
 }]
