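# MetaSound patch for the SoundGenerator tool that produces a vibrato modulation signal, named for FM use.
# VibratoFrequency maps logarithmically to 0.01–1000 Hz and drives a sine oscillator.
# Depth maps through the same log conversion, with BaseFrequency as maximum, and scales the sine so modulation range follows the base pitch.
# Out carries the scaled signal, while NormalizedOut passes the unscaled sine.
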MetaSoundPatch /TechAudioToolsContent/Tools/SoundGenerator/Patches/MSP_SG_FM_Vibrato
inputs [
 BaseFrequency {type=Float}
 Depth {type=Float}
 VibratoFrequency {type=Float}
]
outputs [
 NormalizedOut {type=Audio}
 Out {type=Audio}
]
pages [Default {
  input_defaults {BaseFrequency=0, Depth=0, VibratoFrequency=0}
  nodes [
   LinearToLogFrequency_1 = UE.Linear To Log Frequency(Max Range=^BaseFrequency, Value=^Depth) {also {defaults {Min Range=0}}}
   LinearToLogFrequency_0 = UE.Linear To Log Frequency(Value=^VibratoFrequency) {also {defaults {Max Range=1000, Min Range=0.01}}}
   Sine_0 = UE.Sine:Audio(Frequency=@LinearToLogFrequency_0.Frequency)
   Multiply_0 = UE.Multiply:Audio by Float(AdditionalOperands=@LinearToLogFrequency_1.Frequency, PrimaryOperand=@Sine_0.Audio)
  ]
  links [{dst="@Out", src=Multiply_0.Out}, {dst="@NormalizedOut", src=Sine_0.Audio}]
 }]
