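# MetaSound patch for the Sound Generator tool that adds an inline bitcrusher effect with a dry/wet blend.
# SampleRate passes through a linear-to-log frequency mapping, topping out at 48 kHz, to drive the Bitcrusher's sample rate.
# Amount crossfades between the untouched AudioIn and the crushed signal, and defaults to fully dry.
# Bit depth is not exposed as an input, so only sample-rate reduction is controllable.
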MetaSoundPatch /TechAudioToolsContent/Tools/SoundGenerator/Patches/MSP_SG_FX_InlineBitcrusher
inputs [
 Amount {type=Float}
 AudioIn {type=Audio}
 SampleRate {type=Float}
]
outputs [Out {type=Audio}]
pages [Default {
  input_defaults {Amount=0, AudioIn=null, SampleRate=0}
  nodes [
   LinearToLogFrequency_0 = UE.Linear To Log Frequency(Value=^SampleRate) {also {defaults {Max Range=48000}}}
   CrossfadeTriggerRouteAudio2_0 = "Crossfade.Trigger Route (Audio, 2):Audio"(Crossfade Value=^Amount, In 0=^AudioIn, In 1=UE.Bitcrusher:Audio(Audio=^AudioIn, Sample Rate=@LinearToLogFrequency_0.Frequency).Audio)
  ]
  links [{dst="@Out", src=CrossfadeTriggerRouteAudio2_0.Out}]
 }]
