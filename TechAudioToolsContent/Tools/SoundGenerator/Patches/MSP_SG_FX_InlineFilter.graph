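# MetaSound patch for the SoundGenerator tool that applies a triggered, envelope-swept state variable filter to incoming audio.
# An AD envelope splits Duration between attack and decay by a shape input, with one cubed curve value bending both segments.
# The envelope, scaled by depth, is added to cutoff and converted from linear to log frequency.
# A FilterType input crossfades between low-pass, band-pass and high-pass outputs.
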MetaSoundPatch /TechAudioToolsContent/Tools/SoundGenerator/Patches/MSP_SG_FX_InlineFilter
inputs [
 AudioIn {type=Audio}
 Duration {type=Time}
 FilterCutoff {type=Float}
 FilterEnvelopeCurve {type=Float}
 FilterEnvelopeDepth {type=Float}
 FilterEnvelopeShape {type=Float}
 FilterResonance {type=Float}
 FilterType {type=Float}
 TriggerEnvelope {type=Trigger}
]
outputs [Out {type=Audio}]
pages [Default {
  input_defaults {
   AudioIn=null
   Duration=1
   FilterCutoff=0
   FilterEnvelopeCurve=0
   FilterEnvelopeDepth=1
   FilterEnvelopeShape=0
   FilterResonance=0
   FilterType=0
   TriggerEnvelope=null
  }
  nodes [
   n1 = UE.Power:Float(PrimaryOperand=MapRange.MapRange:Float(In=^FilterEnvelopeCurve) {also {defaults {Out Range A=0.5, Out Range B=1.5}}}.Out Value) {also {defaults {AdditionalOperands=3}}}
   ADEnvelopeADEnvelope_0 = AD Envelope.AD Envelope:Float(
    Trigger=^TriggerEnvelope
    Decay Time=UE.Multiply:Time by Float(
     PrimaryOperand=^Duration
     AdditionalOperands=UE.Subtract:Float(AdditionalOperands=^FilterEnvelopeShape) {also {defaults {PrimaryOperand=1}}}.Out
    ).Out
    Attack Time=UE.Multiply:Time by Float(PrimaryOperand=^Duration, AdditionalOperands=^FilterEnvelopeShape).Out
    Attack Curve=@n1.Out
    Decay Curve=@n1.Out
   )
   LinearToLogFrequency_0 = UE.Linear To Log Frequency(Value=UE.Add:Float(
     AdditionalOperands=^FilterCutoff
     PrimaryOperand=UE.Multiply:Float(AdditionalOperands=^FilterEnvelopeDepth, PrimaryOperand=@ADEnvelopeADEnvelope_0.Out Envelope).Out
    ).Out)
   StateVariableFilter_0 = UE.State Variable Filter:Audio(
    In=^AudioIn
    Cutoff Frequency=@LinearToLogFrequency_0.Frequency
    Resonance=MapRange.MapRange:Float(In=^FilterResonance) {also {defaults {Out Range B=8}}}.Out Value
   )
   CrossfadeTriggerRouteAudio3_0 = "Crossfade.Trigger Route (Audio, 3):Audio"(
    Crossfade Value=MapRange.MapRange:Float(In=^FilterType) {also {defaults {Out Range B=2}}}.Out Value
    In 1=@StateVariableFilter_0.Band Pass
    In 2=@StateVariableFilter_0.High Pass Filter
    In 0=@StateVariableFilter_0.Low Pass Filter
   )
  ]
  links [{dst="@Out", src=CrossfadeTriggerRouteAudio3_0.Out}]
 }]
