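# MetaSound patch for the SoundGenerator tool that applies an inline ring modulation effect to incoming audio, crossfaded from dry to modulated by Amount.
# A sine oscillator, tuned by a log-scaled Frequency input, serves as the ring modulator.
# A triggered attack-decay envelope, scaled by depth, drives the sine's modulation input.
# EnvelopeShape splits Duration between attack and decay, while a cubed remap of EnvelopeCurve bends both segments.
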MetaSoundPatch /TechAudioToolsContent/Tools/SoundGenerator/Patches/MSP_SG_FX_InlineRingmod
inputs [
 Amount {type=Float}
 AudioIn {type=Audio}
 Duration {type=Time}
 EnvelopeCurve {type=Float}
 EnvelopeDepth {type=Float}
 EnvelopeShape {type=Float}
 Frequency {type=Float}
 Trigger {type=Trigger}
]
outputs [Out {type=Audio}]
pages [Default {
  input_defaults {Amount=0, AudioIn=null, Duration=1, EnvelopeCurve=0, EnvelopeDepth=1, EnvelopeShape=1, Frequency=0, Trigger=null}
  nodes [
   LinearToLogFrequency_0 = UE.Linear To Log Frequency(Value=^Frequency) {also {defaults {Max Range=5000}}}
   n1 = UE.Power:Float(PrimaryOperand=MapRange.MapRange:Float(In=^EnvelopeCurve) {also {defaults {Out Range A=0.5, Out Range B=1.5}}}.Out Value) {also {defaults {AdditionalOperands=3}}}
   ADEnvelopeADEnvelope_0 = AD Envelope.AD Envelope:Audio(
    Trigger=^Trigger
    Attack Time=UE.Multiply:Time by Float(PrimaryOperand=^Duration, AdditionalOperands=^EnvelopeShape).Out
    Decay Time=UE.Multiply:Time by Float(PrimaryOperand=^Duration, AdditionalOperands=UE.Subtract:Float(AdditionalOperands=^EnvelopeShape) {also {defaults {PrimaryOperand=1}}}.Out).Out
    Attack Curve=@n1.Out
    Decay Curve=@n1.Out
   )
   n2 = UE.Multiply:Audio by Float(
    AdditionalOperands=^EnvelopeDepth
    PrimaryOperand=MapRange.MapRange:Audio(In=@ADEnvelopeADEnvelope_0.Out Envelope) {also {defaults {In Range A=0, Out Range A=0, Out Range B=5000}}}.Out Value
   )
   CrossfadeTriggerRouteAudio2_0 = "Crossfade.Trigger Route (Audio, 2):Audio"(
    Crossfade Value=^Amount
    In 0=^AudioIn
    In 1=UE.RingMod:Audio(In Carrier=^AudioIn, In Modulator=UE.Sine:Audio(Frequency=@LinearToLogFrequency_0.Frequency, Modulation=@n2.Out).Audio).Out Audio
   )
  ]
  links [{dst="@Out", src=CrossfadeTriggerRouteAudio2_0.Out}]
 }]
