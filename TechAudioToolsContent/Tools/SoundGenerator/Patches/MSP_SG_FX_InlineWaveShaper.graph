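# MSP_SG_FX_InlineWaveShaper is a MetaSound patch in the SoundGenerator tool that applies inline waveshaping distortion to an audio input.
# A single Amount input drives two MapRange nodes, one raising the WaveShaper's shaping amount and the other lowering its output gain as Amount increases.
# This inverse gain mapping offsets the level gained from heavier shaping, and the shaped signal feeds the Out output.
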MetaSoundPatch /TechAudioToolsContent/Tools/SoundGenerator/Patches/MSP_SG_FX_InlineWaveShaper
inputs [
 Amount {type=Float}
 AudioIn {type=Audio}
]
outputs [Out {type=Audio}]
pages [Default {
  input_defaults {Amount=0, AudioIn=null}
  nodes [WaveShaper_0 = UE.WaveShaper:Audio(
    In=^AudioIn
    OutputGain=MapRange.MapRange:Float(In=^Amount) {also {defaults {Out Range A=1.25, Out Range B=0.5}}}.Out Value
    Amount=MapRange.MapRange:Float(In=^Amount) {also {defaults {Out Range A=1, Out Range B=10}}}.Out Value
   )]
  links [{dst="@Out", src=WaveShaper_0.Out}]
 }]
