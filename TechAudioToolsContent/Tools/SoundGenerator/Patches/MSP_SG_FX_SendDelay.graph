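# MetaSound patch that provides the send delay effect for the TechAudioTools Sound Generator.
# It feeds the input audio through a single Delay node, with DelayAmount setting the wet level.
# DelayTime is cubed for a finer low-end response, then mapped against Duration so the delay scales with the sound's length.
# DelayFeedback is remapped with a 0.8 ceiling on the output range to limit feedback buildup.
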MetaSoundPatch /TechAudioToolsContent/Tools/SoundGenerator/Patches/MSP_SG_FX_SendDelay
inputs [
 AudioIn {type=Audio}
 DelayAmount {type=Float}
 DelayFeedback {type=Float}
 DelayTime {type=Float}
 Duration {type=Time}
]
outputs [Out {type=Audio}]
pages [Default {
  input_defaults {AudioIn=null, DelayAmount=0, DelayFeedback=0, DelayTime=0, Duration=0}
  nodes [
   n1 = UE.ConversionTimeToFloat(In=^Duration)
   n2 = UE.Power:Float(PrimaryOperand=^DelayTime) {also {defaults {AdditionalOperands=3}}}
   Delay_0 = UE.Delay:Audio(
    In=^AudioIn
    Wet Level=^DelayAmount
    Delay Time=UE.ConversionFloatToTime(In=MapRange.MapRange:Float(Out Range B=@n1.Out, In=@n2.Out).Out Value).Out
    Feedback=MapRange.MapRange:Float(In=^DelayFeedback) {also {defaults {Out Range B=0.8}}}.Out Value
   )
  ]
  links [{dst="@Out", src=Delay_0.Out}]
 }]
