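# MetaSound patch for the Sound Generator tool that provides a flanger send effect on an audio input.
# The Delay input is remapped and stored in a float variable that sets the center delay of all flangers.
# Four flangers run in series, each with Mix Level at full, and Amount scales the final output with no separate dry path mixed back in.
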MetaSoundPatch /TechAudioToolsContent/Tools/SoundGenerator/Patches/MSP_SG_FX_SendFlanger
inputs [
 Amount {type=Float}
 AudioIn {type=Audio}
 Delay {type=Float}
]
outputs [Out {type=Audio}]
pages [Default {
  input_defaults {Amount=1, AudioIn=null, Delay=0}
  nodes [
   n1 = VariableAccessor.Float(Variable=VariableMutator.Float(
     Variable=InitVariable.Float().Variable
     Value=MapRange.MapRange:Float(In=^Delay) {also {defaults {Out Range A=0.01, Out Range B=2}}}.Out Value
    ).Variable)
   n2 = VariableAccessor.Float(Variable=@n1.Variable)
   n3 = UE.Flanger(In Audio=UE.Flanger(In Audio=^AudioIn, Center Delay=@n1.Value) {also {defaults {Mix Level=1}}}.Out Audio, Center Delay=@n2.Value) {also {defaults {Mix Level=1}}}
   n4 = VariableAccessor.Float(Variable=@n2.Variable)
   Multiply_0 = UE.Multiply:Audio by Float(
    AdditionalOperands=^Amount
    PrimaryOperand=UE.Flanger(
     In Audio=UE.Flanger(In Audio=@n3.Out Audio, Center Delay=@n4.Value) {also {defaults {Mix Level=1}}}.Out Audio
     Center Delay=VariableAccessor.Float(Variable=@n4.Variable).Value
    ) {also {defaults {Mix Level=1}}}.Out Audio
   )
  ]
  links [{dst="@Out", src=Multiply_0.Out}]
 }]
