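# MetaSound patch for the Sound Generator tool that acts as a wet-only plate reverb send effect on a mono signal.
# The mono input feeds both sides of a stereo Plate Reverb, with Amount setting the wet level and the dry level zeroed.
# The left and right reverb outputs are summed and halved into a single mono output, leaving any dry mix to be done outside the patch.
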MetaSoundPatch /TechAudioToolsContent/Tools/SoundGenerator/Patches/MSP_SG_FX_SendReverb
inputs [
 Amount {type=Float}
 AudioIn {type=Audio}
]
outputs [Out {type=Audio}]
pages [Default {
  input_defaults {Amount=1, AudioIn=null}
  nodes [
   PlateReverb_0 = UE.Plate Reverb:Stereo(Wet Level=^Amount, In Left=^AudioIn, In Right=^AudioIn) {also {defaults {Dry Level=0}}}
   Multiply_0 = UE.Multiply:Audio by Float(PrimaryOperand=UE.Add:Audio(PrimaryOperand=@PlateReverb_0.Out Left, AdditionalOperands=@PlateReverb_0.Out Right).Out) {also {defaults {AdditionalOperands=0.5}}}
  ]
  links [{dst="@Out", src=Multiply_0.Out}]
 }]
