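# MetaSound patch for the TechAudioTools SoundGenerator that produces pitched noise by band-pass filtering white noise.
# The FM input is clamped relative to Frequency, converted to float and added, keeping the cutoff between 40 Hz and 18 kHz under modulation.
# An InterpTo smooths that cutoff over a time scaled from Glide.
# A fixed gain boosts the filtered signal before output.
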MetaSoundPatch /TechAudioToolsContent/Tools/SoundGenerator/Patches/MSP_SG_Generator_Noise
inputs [
 FM {type=Audio}
 Frequency {type=Float}
 Glide {type=Float}
]
outputs [Output {type=Audio}]
pages [Default {
  input_defaults {FM=null, Frequency=0, Glide=0}
  nodes [
   n1 = UE.Multiply:Float(PrimaryOperand=^Frequency) {also {defaults {AdditionalOperands=-1}}}
   n2 = UE.Add:Float(PrimaryOperand=@n1.Out) {also {defaults {AdditionalOperands=18000}}}
   n3 = UE.Add:Float(PrimaryOperand=@n1.Out) {also {defaults {AdditionalOperands=40}}}
   n4 = UE.ConversionAudioToFloat(In=Clamp.Clamp:Audio(In=^FM, Max=UE.ConversionFloatToAudio(In=@n2.Out).Out, Min=UE.ConversionFloatToAudio(In=@n3.Out).Out).Value)
   n5 = UE.Multiply:Float(PrimaryOperand=^Glide) {also {defaults {AdditionalOperands=0.25}}}
   BiquadFilter_0 = UE.Biquad Filter:Audio(
    Cutoff Frequency=UE.InterpTo:Audio(Target=UE.Add:Float(PrimaryOperand=^Frequency, AdditionalOperands=@n4.Out).Out, Interp Time=UE.ConversionFloatToTime(In=@n5.Out).Out).Value
    In=UE.Noise:Audio() {also {defaults {Type=1}}}.Audio
   ) {also {defaults {Type=2}}}
   Multiply_0 = UE.Multiply:Audio by Float(PrimaryOperand=@BiquadFilter_0.Out) {also {defaults {AdditionalOperands=3}}}
  ]
  links [{dst="@Output", src=Multiply_0.Out}]
 }]
