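# MetaSound patch that generates a square-wave voice with pulse-width and frequency modulation for the TechAudioTools Sound Generator.
# A saw LFO, with rate derived from Duration, offsets Duty by the DutySweep amount, and the result is remapped to a safe pulse-width range.
# FM input is clamped relative to Frequency to keep modulated pitch within audible limits.
# The oscillator supports Glide, and its output is attenuated by a fixed gain.
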MetaSoundPatch /TechAudioToolsContent/Tools/SoundGenerator/Patches/MSP_SG_Generator_Square
inputs [
 Duration {type=Time}
 Duty {type=Float}
 DutySweep {type=Float}
 FM {type=Audio}
 Frequency {type=Float}
 Glide {type=Float}
]
outputs [Output {type=Audio}]
pages [Default {
  input_defaults {Duration=0.125, Duty=0.5, DutySweep=0, FM=null, Frequency=0, Glide=0}
  nodes [
   n1 = UE.Multiply:Float(PrimaryOperand=^Frequency) {also {defaults {AdditionalOperands=-1}}}
   n2 = UE.ConversionFloatToAudio(In=UE.Add:Float(PrimaryOperand=@n1.Out) {also {defaults {AdditionalOperands=18000}}}.Out)
   n3 = UE.ConversionFloatToAudio(In=UE.Add:Float(PrimaryOperand=@n1.Out) {also {defaults {AdditionalOperands=40}}}.Out)
   n4 = Max.Max:Float(A=UE.ConversionTimeToFloat(In=^Duration).Out) {also {defaults {B=0.01}}}
   n5 = UE.Multiply:Float(PrimaryOperand=^Duty) {also {defaults {AdditionalOperands=-1}}}
   n6 = UE.Subtract:Float(AdditionalOperands=^Duty) {also {defaults {PrimaryOperand=1}}}
   n7 = UE.Add:Float(
    AdditionalOperands=^Duty
    PrimaryOperand=UE.LFO:Audio(
     Frequency=UE.Divide:Float(AdditionalOperands=@n4.Value).Out
     Min Value=UE.Multiply:Float(PrimaryOperand=^DutySweep, AdditionalOperands=@n5.Out).Out
     Max Value=UE.Multiply:Float(PrimaryOperand=^DutySweep, AdditionalOperands=@n6.Out).Out
    ) {also {defaults {Shape=1}}}.Out
   )
   Multiply_0 = UE.Multiply:Audio by Float(PrimaryOperand=UE.Square:Audio(
     Frequency=^Frequency
     Glide=^Glide
     Modulation=Clamp.Clamp:Audio(In=^FM, Max=@n2.Out, Min=@n3.Out).Value
     Pulse Width=MapRange.MapRange:Float(In=@n7.Out) {also {defaults {Out Range A=0.05, Out Range B=0.95}}}.Out Value
    ).Audio) {also {defaults {AdditionalOperands=0.35}}}
  ]
  links [{dst="@Output", src=Multiply_0.Out}]
 }]
