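# MetaSound patch providing the triangle wave generator for the TechAudioTools Sound Generator, outputting mono audio.
# Frequency and Glide drive a Triangle oscillator, with audio-rate FM feeding its modulation input.
# A clamp bounds FM by offsets derived from the negated Frequency, keeping the modulated pitch between 40 Hz and 18 kHz.
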MetaSoundPatch /TechAudioToolsContent/Tools/SoundGenerator/Patches/MSP_SG_Generator_Triangle
inputs [
 FM {type=Audio}
 Frequency {type=Float}
 Glide {type=Float}
]
outputs [Output {type=Audio}]
pages [Default {
  input_defaults {FM=null, Frequency=0, Glide=0}
  nodes [
   n1 = UE.Multiply:Float(PrimaryOperand=^Frequency) {also {defaults {AdditionalOperands=-1}}}
   n2 = UE.Add:Float(PrimaryOperand=@n1.Out) {also {defaults {AdditionalOperands=18000}}}
   n3 = UE.Add:Float(PrimaryOperand=@n1.Out) {also {defaults {AdditionalOperands=40}}}
   Triangle_0 = UE.Triangle:Audio(
    Frequency=^Frequency
    Glide=^Glide
    Modulation=Clamp.Clamp:Audio(In=^FM, Max=UE.ConversionFloatToAudio(In=@n2.Out).Out, Min=UE.ConversionFloatToAudio(In=@n3.Out).Out).Value
   ) {also {defaults {Type=0}}}
  ]
  links [{dst="@Output", src=Triangle_0.Audio}]
 }]
