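# MetaSound patch for the Sound Generator tool that outputs a stepped float for a multi-stage pitch jump.
# OnPlay starts a Trigger Repeat that fires NumberOfStages plus one times, spaced evenly by dividing StageDuration by that count, floored at one.
# Each repeat advances a Trigger Counter that starts at 1, steps by PitchInterval and resets at the same count, driving the Value output.
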MetaSoundPatch /TechAudioToolsContent/Tools/SoundGenerator/Patches/MSP_SG_MultiStageJump
inputs [
 NumberOfStages {type=Int32}
 OnPlay {type=Trigger}
 PitchInterval {type=Float}
 StageDuration {type=Time}
]
outputs [Value {type=Float}]
pages [Default {
  input_defaults {NumberOfStages=0, OnPlay=null, PitchInterval=0, StageDuration=0}
  nodes [
   n1 = UE.Add:Int32(PrimaryOperand=^NumberOfStages) {also {defaults {AdditionalOperands=1}}}
   n2 = UE.Reroute(Value=UE.Reroute(Value=@n1.Out).Value)
   n3 = Max.Max:Float(A=Convert.Float:Int32(Int32=@n1.Out).Float) {also {defaults {B=1}}}
   TriggerCounter_0 = UE.Trigger Counter(
    Step Size=^PitchInterval
    Reset Count=UE.Reroute(Value=@n2.Value).Value
    In=UE.TriggerRepeat(
     Start=^OnPlay
     Period=UE.Divide:Time by Float(PrimaryOperand=^StageDuration, AdditionalOperands=@n3.Value) {also {version [1,1]}}.Out
     Num Repeats=@n2.Value
    ).RepeatOut
   ) {also {defaults {Start Value=1}}}
  ]
  links [{dst="@Value", src=TriggerCounter_0.Value}]
 }]
